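# Translucent unlit Niagara sprite material for the Galaxy Knight pickaxe FX, drawing round smoke puffs filled with a starry nebula.
# Triplanar camera-vector lookups build the rotating nebula and two bright star layers, one modulated by panning noise, tinted by particle color.
# Opacity subtracts radially panned noise and an optional inner mask from a sphere mask, using original-UV derivatives for mip selection.
# Particle alpha scales opacity, with depth fade skipped on mobile.
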Material /BRCosmetics/Weapons/FORT_Melee/Pickaxe_Galaxy_Knight_Male/FX/M_GalaxySmoke
graph {
 properties {bUsedWithNiagaraSprites=true, blendMode=BLEND_Translucent, pixelDepthOffsetMode=PDOM_Legacy, shadingModel=MSM_Unlit, usedShadingModels=Unlit}
 nodes [
  n1 = Time()
  n2 = Multiply(A=@n1, B=Constant())
  n3 = ScalarParameter() {group=Wrap_Galaxy, parameterName=WG_Background Rotation}
  n4 = AppendVector(
   A=ScalarParameter() {defaultValue=0.02, group=Wrap_Galaxy, parameterName=WG_Galaxy Rotation Speed}
   B=ScalarParameter() {defaultValue=-0.02, group=Wrap_Galaxy, parameterName=WG_Galaxy Rotation Speed_1}
  )
  n5 = ScalarParameter() {defaultValue=0.1, group=Wrap_Galaxy, parameterName=WG_Galaxy Rotation Speed_2}
  n6 = TextureObjectParameter() {
   group=Wrap_Galaxy
   parameterName=WG_GalaxyTexture
   texture=/Game/Characters/Player/Male/Medium/Bodies/M_Med_Soldier_04/Skins/Galaxy/Textures/T_FN_Nebula
  }
  n7 = MaterialFunctionCall!(
   CameraVector
   Rotation Speed=Multiply(A=LinearInterpolate(A=@n1, B=@n2, Alpha=@n3), B=AppendVector(A=@n4, B=@n5))
   Texture=@n6
   Tiling=ScalarParameter() {defaultValue=0.5, group=Wrap_Galaxy, parameterName=WG_GalaxyTiling}
   Offset=Multiply(B=ScalarParameter() {defaultValue=0.012, group=Wrap_Galaxy, parameterName=WG_WarpIntensity})
   AxisFadeContrast=ScalarParameter() {defaultValue=6, group=Wrap_Galaxy, parameterName=WG_GalaxyAxisFade}
  ) {materialFunction=/Game/Weapons/WeaponSkins/MaterialFunction/TriplanarCameraVector_Rotated}
  n8 = LinearInterpolate(
   A=@n7.Result
   B=SmoothStep(
    Min=ScalarParameter() {group=Wrap_Galaxy, parameterName=WG_SSMin}
    Max=ScalarParameter() {group=Wrap_Galaxy, parameterName=WG_SSMax}
    Value=@n7.Result
   )
   Alpha=ScalarParameter() {group=Wrap_Galaxy, parameterName=WG_UseSmoothstep}
  )
  n9 = Desaturation(None=@n8, Fraction=ScalarParameter() {defaultValue=0.5, group=Wrap_Galaxy, parameterName=WG_Star Saturation})
  n10 = VectorParameter() {group=Wrap_Galaxy, parameterName=WG_Star Brightness, defaultValue=<rgba 300 300 300 1>}
  n11 = Time()
  n12 = ScalarParameter() {defaultValue=-0.125, group=Wrap_Galaxy, parameterName=WG_Small Star Rotation speed}
  n13 = TextureObjectParameter() {
   parameterName=Stars
   samplerType=SAMPLERTYPE_Masks
   texture=/Game/Characters/Player/Male/Medium/Bodies/M_Med_Soldier_04/Skins/Galaxy/Textures/Celestial_T_Stars
  }
  n14 = Time()
  n15 = VectorParameter() {group=Wrap_Galaxy, parameterName=WG_Small Star_Pan_Speed, defaultValue=<rgba 0 0 0 1>}
  n16 = ComponentMask(None=MaterialFunctionCall(
    Rotation Speed=Multiply(A=@n11, B=@n12)
    Texture=@n13
    Tiling=ScalarParameter() {defaultValue=5, group=Wrap_Galaxy, parameterName=WG_Small Star Tiling}
    Offset=Multiply(A=@n14, B=@n15.RGB)
   ) {materialFunction=&TriplanarCameraVector_Rotated}.Result) {b=true}
  n17 = ComponentMask(None=MaterialFunctionCall!(
    CameraVector
    Texture=@n13
    Tiling=ScalarParameter() {defaultValue=2, group=Wrap_Galaxy, parameterName=WG_Bright Star Tiling}
    Offset
    AxisFadeContrast
   ) {materialFunction=/Engine/Functions/Engine_MaterialFunctions01/Texturing/TriplanarCameraVector}.Result) {r=true}
  n18 = TextureSample(UVs=Panner(Coordinate=TextureCoordinate() {uTiling=2, vTiling=2}) {speedX=0.133, speedY=0.2}) {
   samplerSource=SSM_Wrap_WorldGroupSettings
   samplerType=SAMPLERTYPE_Masks
   texture=&Celestial_T_Stars
  }
  n19 = LinearInterpolate(
   A=LinearInterpolate(A=@n8, B=Multiply(A=@n9, B=@n10.RGB), Alpha=Multiply(A=@n16, B=@n16))
   B=Multiply(A=@n9, B=VectorParameter() {group=Wrap_Galaxy, parameterName=WG_Star Brightness_1, defaultValue=<rgba 950 950 950 1>}.RGB)
   Alpha=Multiply(A=Multiply(A=@n17, B=@n17), B=@n18.G)
  )
  n20 = ScalarParameter() {parameterName=GlobalDesat}
  n21 = ParticleColor()
  MaterialFunctionCall(Emissive=Multiply(A=Desaturation(None=@n19, Fraction=@n20), B=@n21.RGB)) {materialFunction=&MF_FX_EmissiveAdjustments}
  n22 = TextureCoordinate()
  n23 = Multiply(
   A=SphereMask(A=@n22, B=Constant() {r=0.5}) {attenuationRadius=0.5, hardnessPercent=0}
   B=ScalarParameter() {defaultValue=1.5, parameterName=Roundness Strength, sliderMax=5}
  ) {constB=1.5}
  n24 = MaterialFunctionCall(Vector or UVs=ConstantBiasScale(None=TextureCoordinate()) {bias=-0.5, scale=2}) {materialFunction=&VectorToRadialValue}
  n25 = AppendVector(
   A=ScalarParameter() {defaultValue=2, parameterName=Tiling X, sliderMax=5}
   B=ScalarParameter() {defaultValue=1, parameterName=Tiling Y, sliderMax=5}
  )
  n26 = ScalarParameter() {parameterName=Speed X, sliderMax=2, sliderMin=-2}
  n27 = ScalarParameter() {defaultValue=-0.5, parameterName=Speed Y, sliderMax=2, sliderMin=-2}
  n28 = TextureCoordinate()
  n29 = TextureSample!(
   UVs=Panner(Coordinate=Multiply(A=@n24.Radial Coordinates, B=@n25), Speed=AppendVector(A=@n26, B=@n27)) {speedY=-1}
   Tex
   "DDX(UVs)"=DDX(Value=@n28)
   "DDY(UVs)"=DDY(Value=@n28)
   Apply View MipBias
  ) {mipValueMode=TMVM_Derivative, samplerType=SAMPLERTYPE_Masks, texture=/FortWater/Textures/OceanNoise__Mask}
  n30 = Multiply(
   A=SphereMask(
    A=@n22
    B=Constant() {r=0.5}
    Radius=ScalarParameter() {defaultValue=0.2, parameterName=Inner Mask Radius}
    Hardness=ScalarParameter() {defaultValue=0.5, parameterName=Inner Mask Hardness}
   ) {attenuationRadius=0.2, hardnessPercent=0}
   B=ScalarParameter() {defaultValue=1.5, parameterName=Inner Mask Strength, sliderMax=5}
  ) {constB=1.5}
  n31 = ScalarParameter() {parameterName=Use Inner Mask}
  n32 = Multiply(A=@n21.A, B=Subtract(A=Subtract(A=@n23, B=@n29.R), B=LinearInterpolate(B=@n30, Alpha=@n31) {constAlpha=0}))
  n33 = DepthFade(Opacity=@n32, FadeDistance=ScalarParameter() {defaultValue=100, parameterName=Depth Fade Distance, sliderMax=200})
  ShadingPathSwitch(Default=@n33, Deferred=@n33, Mobile=@n32) {desc=Migrated from FeatureLevelSwitch}
 ]
}
