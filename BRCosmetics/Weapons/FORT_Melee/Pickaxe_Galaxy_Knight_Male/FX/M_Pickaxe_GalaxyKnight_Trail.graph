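# Translucent, unlit, two-sided Niagara ribbon and mesh-particle material for the Galaxy Knight pickaxe trail.
# Emissive blends particle color into a galaxy of camera-vector triplanar nebula and star layers, with a rotatable sheen-swipe gradient.
# Opacity erodes an edge-faded UV shape with Voronoi-distorted, panning fire noise, plus camera and depth fades.
# Layered sines ripple vertices along the trail; dynamic parameters drive UV offset, camera offset and overall opacity.
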Material /BRCosmetics/Weapons/FORT_Melee/Pickaxe_Galaxy_Knight_Male/FX/M_Pickaxe_GalaxyKnight_Trail
graph {
 properties {
  bUsedWithNiagaraMeshParticles=true
  bUsedWithNiagaraRibbons=true
  blendMode=BLEND_Translucent
  pixelDepthOffsetMode=PDOM_Legacy
  shadingModel=MSM_Unlit
  twoSided=true
  usedShadingModels=Unlit
 }
 nodes [
  n1 = DynamicParameter() {defaultValue=<rgba 0.5 1 0 1>, paramNames [UV1 U Offset, Emissive Switch, Camera Offset, Overall opacity]}
  n2 = DynamicParameter() {defaultValue=<rgba 0.5 1 0 1>, paramNames [UV1 U Offset, Emissive Switch, Camera Offset, Overall opacity]}
  n3 = Add(A=TextureCoordinate(), B=AppendVector(A=@n2.UV1 U Offset, B=Constant()))
  n4 = ComponentMask(None=Panner(Coordinate=@n3) {speedX=-1}) {r=true}
  n5 = Sine(None=ComponentMask(None=Panner(Coordinate=@n3) {speedX=-0.5}) {r=true}) {period=0.5}
  n6 = ComponentMask(None=Panner() {speedX=-0.75}) {r=true}
  n7 = Add(A=Add(A=Sine(None=@n4) {period=1.4}, B=Multiply(A=@n5) {constB=0.3}), B=Multiply(A=Sine(None=@n6) {period=3}) {constB=0.7})
  n8 = Transform(None=VectorParameter() {group=WPO, parameterName=WPO Direction, defaultValue=<rgba 1 1 0 1>}.RGB) {transformSourceType=TRANSFORMSOURCE_Instance}
  n9 = ComponentMask(None=TextureCoordinate()) {r=true}
  Add(
   A=Multiply(
    A=MaterialFunctionCall() {materialFunction=&CameraDirectionVector}.Result
    B=@n1.Camera Offset
   )
   B=Multiply(A=Multiply(A=@n7, B=@n8), B=Power(Base=@n9))
  )
  n10 = Time()
  n11 = Multiply(A=@n10, B=Constant())
  n12 = ScalarParameter() {group=Wrap_Galaxy, parameterName=WG_Background Rotation}
  n13 = AppendVector(
   A=ScalarParameter() {defaultValue=0.02, group=Wrap_Galaxy, parameterName=WG_Galaxy Rotation Speed}
   B=ScalarParameter() {defaultValue=-0.02, group=Wrap_Galaxy, parameterName=WG_Galaxy Rotation Speed_1}
  )
  n14 = ScalarParameter() {defaultValue=0.1, group=Wrap_Galaxy, parameterName=WG_Galaxy Rotation Speed_2}
  n15 = TextureObjectParameter() {
   group=Wrap_Galaxy
   parameterName=WG_GalaxyTexture
   texture=/Game/Characters/Player/Male/Medium/Bodies/M_Med_Soldier_04/Skins/Galaxy/Textures/T_FN_Nebula
  }
  n16 = MaterialFunctionCall!(
   CameraVector
   Rotation Speed=Multiply(A=LinearInterpolate(A=@n10, B=@n11, Alpha=@n12), B=AppendVector(A=@n13, B=@n14))
   Texture=@n15
   Tiling=ScalarParameter() {defaultValue=0.5, group=Wrap_Galaxy, parameterName=WG_GalaxyTiling}
   Offset=Multiply(B=ScalarParameter() {defaultValue=0.012, group=Wrap_Galaxy, parameterName=WG_WarpIntensity})
   AxisFadeContrast=ScalarParameter() {defaultValue=6, group=Wrap_Galaxy, parameterName=WG_GalaxyAxisFade}
  ) {materialFunction=/Game/Weapons/WeaponSkins/MaterialFunction/TriplanarCameraVector_Rotated}
  n17 = Desaturation(None=@n16.Result, Fraction=ScalarParameter() {defaultValue=0.5, group=Wrap_Galaxy, parameterName=WG_Star Saturation})
  n18 = Time()
  n19 = ScalarParameter() {defaultValue=-0.125, group=Wrap_Galaxy, parameterName=WG_Small Star Rotation speed}
  n20 = TextureObjectParameter() {
   parameterName=Stars
   samplerType=SAMPLERTYPE_Masks
   texture=/Game/Characters/Player/Male/Medium/Bodies/M_Med_Soldier_04/Skins/Galaxy/Textures/Celestial_T_Stars
  }
  n21 = Time()
  n22 = VectorParameter() {group=Wrap_Galaxy, parameterName=WG_Small Star_Pan_Speed, defaultValue=<rgba 0 0 0 1>}
  n23 = ComponentMask(None=MaterialFunctionCall(
    Rotation Speed=Multiply(A=@n18, B=@n19)
    Texture=@n20
    Tiling=ScalarParameter() {defaultValue=5, group=Wrap_Galaxy, parameterName=WG_Small Star Tiling}
    Offset=Multiply(A=@n21, B=@n22.RGB)
   ) {materialFunction=&TriplanarCameraVector_Rotated}.Result) {b=true}
  n24 = LinearInterpolate(
   A=@n16.Result
   B=Multiply(A=@n17, B=VectorParameter() {group=Wrap_Galaxy, parameterName=WG_Star Brightness, defaultValue=<rgba 300 300 300 1>}.RGB)
   Alpha=Multiply(A=@n23, B=@n23)
  )
  n25 = Multiply(A=@n17, B=VectorParameter() {group=Wrap_Galaxy, parameterName=WG_Star Brightness_1, defaultValue=<rgba 950 950 950 1>}.RGB)
  n26 = ComponentMask(None=MaterialFunctionCall!(
    CameraVector
    Texture=@n20
    Tiling=ScalarParameter() {defaultValue=2, group=Wrap_Galaxy, parameterName=WG_Bright Star Tiling}
    Offset
    AxisFadeContrast
   ) {materialFunction=/Engine/Functions/Engine_MaterialFunctions01/Texturing/TriplanarCameraVector}.Result) {r=true}
  n27 = TextureSample(UVs=Panner(Coordinate=TextureCoordinate() {uTiling=2, vTiling=2}) {speedX=0.133, speedY=0.2}) {
   samplerSource=SSM_Wrap_WorldGroupSettings
   samplerType=SAMPLERTYPE_Masks
   texture=&Celestial_T_Stars
  }
  n28 = Multiply(A=Multiply(A=@n26, B=@n26), B=@n27.G)
  MaterialFunctionCall(Emissive=LinearInterpolate(A=ParticleColor().RGB, B=LinearInterpolate(A=@n24, B=@n25, Alpha=@n28), Alpha=NamedRerouteUsage())) {materialFunction=&MF_FX_EmissiveAdjustments}
  n29 = ScalarParameter() {defaultValue=0.1, parameterName=Leading Edge Fade, sliderMax=1}
  n30 = TextureCoordinate()
  n31 = ComponentMask(None=@n30) {r=true}
  n32 = OneMinus(None=@n31)
  n33 = ComponentMask(None=@n30) {g=true}
  n34 = Multiply(A=SmoothStep(Max=@n29, Value=@n31) {constMax=0.1}, B=Multiply(A=Power(Base=@n32), B=@n33))
  n35 = SmoothStep(Max=@n29, Value=OneMinus(None=@n33)) {constMax=0.1}
  n36 = DynamicParameter() {defaultValue=<rgba 0.5 1 0 1>, paramNames [UV1 U Offset, Emissive Switch, Camera Offset, Overall opacity]}
  n37 = Constant()
  n38 = Panner(
   Coordinate=TextureCoordinate() {coordinateIndex=1}
   Speed=AppendVector(
    A=ScalarParameter() {defaultValue=-2, group=UV Distortion, parameterName=UV Speed X 2, sliderMax=5, sliderMin=-5}
    B=ScalarParameter() {group=UV Distortion, parameterName=UV Speed Y 2, sliderMax=5, sliderMin=-5}
   )
  ) {speedX=-0.5}
  n39 = AppendVector(
   A=ScalarParameter() {defaultValue=0.25, group=UV Distortion, parameterName=UV Scale X 2, sliderMax=5}
   B=ScalarParameter() {defaultValue=1, group=UV Distortion, parameterName=UV Scale Y 2, sliderMax=5}
  )
  n40 = TextureSampleParameter2D(UVs=Add(A=AppendVector(A=@n36.UV1 U Offset, B=@n37), B=Multiply(A=@n38, B=@n39))) {group=UV Distortion, parameterName=Distortion Texture, texture=/Game/Effects/Fort_Effects/Textures/T_PinkSpike_Voronoi2D_Height}
  n41 = DynamicParameter() {defaultValue=<rgba 0.5 1 0 1>, paramNames [UV1 U Offset, Emissive Switch, Camera Offset, Overall opacity]}
  n42 = Constant()
  n43 = Panner(
   Coordinate=TextureCoordinate() {coordinateIndex=1}
   Speed=AppendVector(
    A=ScalarParameter() {defaultValue=-1, group=UV Distortion, parameterName=UV Speed X, sliderMax=5, sliderMin=-5}
    B=ScalarParameter() {group=UV Distortion, parameterName=UV Speed Y, sliderMax=5, sliderMin=-5}
   )
  ) {speedX=-0.5}
  n44 = AppendVector(
   A=ScalarParameter() {defaultValue=0.25, group=UV Distortion, parameterName=UV Scale X, sliderMax=5}
   B=ScalarParameter() {defaultValue=1, group=UV Distortion, parameterName=UV Scale Y, sliderMax=5}
  )
  n45 = TextureSampleParameter2D(UVs=Add(A=AppendVector(A=@n41.UV1 U Offset, B=@n42), B=Multiply(A=@n43, B=@n44))) {group=UV Distortion, parameterName=Distortion Texture, texture=&T_PinkSpike_Voronoi2D_Height}
  n46 = Subtract(A=MaterialFunctionCall(X=Multiply(A=@n40.R, B=@n45.R), Y=Multiply(A=@n40.G, B=@n45.G)) {materialFunction=&MakeFloat2}.Result) {constB=0.25}
  n47 = ScalarParameter() {defaultValue=0.1, group=UV Distortion, parameterName=Distortion Strength, sliderMax=1, sliderMin=-1}
  n48 = Panner(
   Coordinate=Add(A=TextureCoordinate() {coordinateIndex=1}, B=Multiply(A=@n46, B=@n47) {constB=0.1})
   Speed=AppendVector(
    A=ScalarParameter() {defaultValue=-1, group=Flame, parameterName=Speed X, sliderMax=5, sliderMin=-5}
    B=ScalarParameter() {group=Flame, parameterName=Speed Y, sliderMax=5, sliderMin=-5}
   )
  ) {speedX=1}
  n49 = AppendVector(
   A=ScalarParameter() {defaultValue=0.25, group=Flame, parameterName=Scale X, sliderMax=5}
   B=ScalarParameter() {defaultValue=1, group=Flame, parameterName=Scale Y, sliderMax=5}
  )
  n50 = DynamicParameter() {defaultValue=<rgba 0.5 1 0 1>, paramNames [UV1 U Offset, Emissive Switch, Camera Offset, Overall opacity]}
  n51 = Constant()
  n52 = TextureSampleParameter2D(UVs=Add(A=Multiply(A=@n48, B=@n49), B=AppendVector(A=@n50.UV1 U Offset, B=@n51))) {parameterName=Mask Texture, texture=/Game/Effects/Fort_Effects/Textures/T_WispyFireNoise03}
  n53 = Subtract(A=Multiply(A=Multiply(A=@n34, B=@n35), B=ScalarParameter() {defaultValue=2, parameterName=Shape Hardness, sliderMax=5}) {constB=3}, B=@n52.R)
  n54 = ParticleColor()
  n55 = ScalarParameter() {defaultValue=100, parameterName=Camera Fade Length, sliderMax=200}
  n56 = ScalarParameter() {defaultValue=25, parameterName=Camera Fade Offset, sliderMax=200}
  n57 = SmoothStep(
   Max=OneMinus(None=ScalarParameter() {defaultValue=0.2, parameterName=Edge Hardness, sliderMax=1, sliderMin=-1})
   Value=Multiply(
    A=Multiply(A=@n53, B=@n54.A)
    B=MaterialFunctionCall(Fade Length=@n55, Fade Offset=@n56) {materialFunction=&CameraDepthFade}.Result
   )
  ) {constMax=2}
  n58 = DepthFade(Opacity=@n57, FadeDistance=ScalarParameter() {defaultValue=35, parameterName=Depth Fade Distance, sliderMax=100}) {fadeDistanceDefault=25}
  n59 = DynamicParameter() {defaultValue=<rgba 0.5 1 0 1>, paramNames [UV1 U Offset, Emissive Switch, Camera Offset, Overall opacity]}
  Multiply(
   A=Multiply(A=ShadingPathSwitch(Default=@n58, Deferred=@n58, Forward=@n57, Mobile=@n57), B=@n59.Overall opacity)
   B=ScalarParameter() {defaultValue=1, parameterName=Opacity, sliderMax=1}
  )
  n60 = Add(A=TextureCoordinate(), B=Multiply(A=@n46) {constB=0})
  n61 = ComponentMask(None=@n60) {r=true}
  n62 = Multiply(A=AppendVector(A=OneMinus(None=@n61), B=ComponentMask(None=@n60) {g=true}))
  n63 = Constant2Vector() {g=0.5, r=0.5}
  Multiply(A=Multiply(A=SphereMask(A=@n62, B=@n63) {attenuationRadius=0.5, hardnessPercent=75}) {constB=3}, B=@n52.R)
  n64 = MaterialFunctionCall(Float2=Subtract(A=TextureCoordinate()) {constB=0.5}) {materialFunction=&BreakOutFloat2Components}
  n65 = DynamicParameter() {parameterIndex=1, defaultValue=<rgba 0 0.3 0.03 0>, paramNames [Sheen Sweep, Sheen Size, Sheen Rotation, - unused -]}
  n66 = Cosine(None=@n65.Sheen Rotation)
  n67 = Sine(None=@n65.Sheen Rotation)
  n68 = AppendVector(A=Add(A=Multiply(A=@n64.R, B=@n66), B=Multiply(A=@n64.G, B=@n67)), B=Subtract(A=Multiply(A=@n64.G, B=@n66), B=Multiply(A=@n64.R, B=@n67)))
  n69 = Multiply(A=@n65.Sheen Rotation) {constB=4}
  n70 = Abs(None=ConstantBiasScale(None=Frac(None=@n69)) {bias=-0.5, scale=2})
  n71 = DotProduct(A=@n70, B=@n70)
  n72 = SquareRoot(None=Constant() {r=2})
  n73 = Add(A=Multiply(A=OneMinus(None=@n71), B=Subtract(A=@n72)))
  n74 = Add(A=VertexInterpolator(VS=Divide(A=@n68, B=@n73)).PS) {constB=0.5}
  n75 = Multiply(A=Time()) {constB=0.333}
  n76 = DynamicParameter() {parameterIndex=1, defaultValue=<rgba 0 0.3 0.03 0>, paramNames [Sheen Sweep, Sheen Size, Sheen Rotation, - unused -]}
  n77 = ParticleRelativeTime()
  n78 = Saturate(None=LinearInterpolate(A=Frac(None=@n75) {desc=Sheen Position}, B=@n76.Sheen Sweep, Alpha=Ceil(None=@n77)) {desc="Will swap to Particle Parameter\r\nwhen inside a partlce system"})
  n79 = Add(A=@n76.Sheen Size)
  n80 = Subtract(A=OneMinus(None=ComponentMask(None=@n74) {r=true}), B=Subtract(A=Multiply(A=@n78, B=@n79), B=@n76.Sheen Size))
  NamedRerouteDeclaration(None=Saturate(None=Divide(A=@n80, B=@n76.Sheen Size)) {desc=Remapped gradient}) {name=Swipe, nodeColor=<rgba 1 0 0.8672 1>}
  ScalarParameter() {parameterName=Camera offset}
 ]
}
