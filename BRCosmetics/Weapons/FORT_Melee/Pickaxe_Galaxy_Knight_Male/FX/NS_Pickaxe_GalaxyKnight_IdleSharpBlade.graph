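# Niagara idle effect for the Galaxy Knight pickaxe's sharp blade: a mesh sheen, a timed sparkle and a star-map trail.
# Sheen lifetime minus an offset times the sparkle burst and paces the stars.
# Eight stars take positions from a system array, mirrored in Y on alternate loops except the last, drawing as shrinking sprites plus ribbon.
# An interpolated User.bisSwinging value gates star alpha; all but the last star drift in randomized directions.
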NiagaraSystem /BRCosmetics/Weapons/FORT_Melee/Pickaxe_Galaxy_Knight_Male/FX/NS_Pickaxe_GalaxyKnight_IdleSharpBlade
dependencies {
 usedDataInterfaces [
  &NiagaraDataInterfaceCurve
  &NiagaraDataInterfaceArrayFloat
  &NiagaraDataInterfaceColorCurve
  &NiagaraDataInterfaceArrayFloat3
 ]
 usedDynamicInputs []
 usedModules [
  &SystemState
  &self:SystemSpawnScript.SystemScriptSource.SystemScriptGraph.NiagaraNodeAssignment_1.SetVariables_#1
  &self:SystemSpawnScript.SystemScriptSource.SystemScriptGraph.NiagaraNodeAssignment_2.SetVariables_#2
  &self:SystemSpawnScript.SystemScriptSource.SystemScriptGraph.NiagaraNodeAssignment_3.SetVariables_#3
  &self:SystemSpawnScript.SystemScriptSource.SystemScriptGraph.NiagaraNodeAssignment_4.SetVariables_#4
 ]
}
emitters {
 Sheen_Mesh {
  emitterUpdateScript [
   &EmitterState()
   &SpawnBurst_Instantaneous(Use Spawn Probability=false, UseLoopCountLimit=false)
  ]
  particleSpawnScript [
   &InitializeParticle(Color=<rgba 0.4341 0.2487 0.646 0.5>, Lifetime=$System.SheenLifetime, UsePositionOffset=false, Write Material Random=true)
   &DynamicMaterialParameters(
    Index 0 Param 2=0.4
    Index 0 Param 3=-0.02
    Module.Param0WriteEnabled=false
    Module.Param1WriteEnabled=true
    Module.Param2WriteEnabled=true
    Module.Param3WriteEnabled=false
   )
  ]
  particleUpdateScript [
   &ParticleState()
   &DynamicMaterialParameters(
    Index 0 Param 1=&FloatFromCurve(FloatCurve=DataInterface() {propertyValues {Curve {keys [
         {arriveTangent=-0.0128, interpMode=RCIM_Cubic, leaveTangent=-0.0128, tangentMode=RCTM_User, time=0, value=0}
         {arriveTangent=0.0436, interpMode=RCIM_Cubic, leaveTangent=0.0436, tangentMode=RCTM_User, time=1, value=0.8592}
        ]}}})
    Module.Param0WriteEnabled=true
    Module.Param1WriteEnabled=false
    Module.Param2WriteEnabled=false
    Module.Param3WriteEnabled=false
   ) {module=DynamicMaterialParameters001}
  ]
  rendererClasses [&NiagaraMeshRendererProperties]
  renderers [{rendererClass=&NiagaraMeshRendererProperties, rendererIndex=0}]
 }
 Sparkle {
  emitterUpdateScript [
   &EmitterState()
   &SpawnBurst_Instantaneous(
    Spawn Time=&Subtract_Float(A=$System.SheenLifetime, B=$System.SparkleSpawnTimeOffset)
    Use Spawn Probability=false
    UseLoopCountLimit=false
   )
  ]
  particleSpawnScript [&InitializeParticle(
    Color=<rgba 2 0.708 2 1>
    Lifetime=&RandomRangeFloat(Minimum=0.666, Maximum=0.777)
    Position Offset=$System.SparklePositionOffset
    Sprite Rotation Mode=::NewEnumerator2
    Sprite Size=&Multiply_Vector2D_ByFloat(Float=&RandomRangeFloat(Minimum=20, Maximum=25), Vector2D=<1.1 1>)
    Sprite Size Mode=::NewEnumerator4
    UsePositionOffset=true
    Write Material Random=true
   )]
  particleUpdateScript [
   &ParticleState()
   &ScaleSpriteSize(Uniform Curve Sprite Scale=DataInterface() {propertyValues {Curve {keys [
        {arriveTangent=9.0725, interpMode=RCIM_Cubic, leaveTangent=9.0725, tangentMode=RCTM_User, time=0, value=0}
        {arriveTangent=-1.7917, interpMode=RCIM_Cubic, leaveTangent=-1.7917, tangentMode=RCTM_User, time=0.0666, value=0.9847}
        {arriveTangent=-0.0405, interpMode=RCIM_Cubic, leaveTangent=-0.0405, tangentMode=RCTM_User, time=1, value=0}
       ]}}})
   &SpriteRotationRate(Rotation Rate=&RandomRangeFloat(Minimum=90, Maximum=360))
   &CameraOffset(Camera Offset Amount=1)
  ]
  rendererClasses [&NiagaraSpriteRendererProperties]
  renderers [{rendererClass=&NiagaraSpriteRendererProperties}]
 }
 Star_Map_Loading {
  emitterSpawnScript [/SetVariables(Emitter.Star Amount=8) {isSetParameters=true}]
  emitterUpdateScript [
   &EmitterState()
   &SpawnRate(
    SpawnRate=&Multiply_Float(
     A=&Divide_Float(
      A=&Multiply_Float(A=1, B=$Emitter.Star Amount)
      B=&Subtract_Float(A=$System.SheenLifetime, B=$System.SparkleSpawnTimeOffset)
     )
     B=&MakeCustomFloatFromBool(
      Bool=&SetBoolByFloatComparison(
       A=&MakeFloatFromInt(Int=$Emitter.SpawnedAmount)
       B=&Multiply_Float(
        A=$Emitter.Star Amount
        B=&MakeFloatFromInt(Int=&Add_Integer(A=$Emitter.LoopCount, B=1))
       )
       Comparison Type=::NewEnumerator6
      )
      True Float=1
      False Float=0
     )
    )
    Use Spawn Probability=false
   )
   /SetVariables(Emitter.SpawnedAmount:NiagaraInt32=&Add_Integer(
     A=$Emitter.SpawnedAmount
     B=&MakeCustomIntFromBool(Bool=$Output.SpawnRate.HasSpawnedThisFrame, True INT=1, False INT=0)
    )) {isSetParameters=true}
  ]
  particleSpawnScript [
   &PartitionParticles(
    ID to Partition:NiagaraInt32=/Niagara/DynamicInputs/Subtract/Subtract_Integer(A:NiagaraInt32=$Emitter.SpawnedAmount, B:NiagaraInt32=1)
    Number Of Partitions=&MakeIntFromFloat(Float=$Emitter.Star Amount, Conversion Mode=::NewEnumerator1)
   )
   &InitializeParticle(
    Color=<rgba 1 1 1 0.5>
    Lifetime=&Subtract_Float(A=1.5, B=$System.LoopedAge)
    Position Offset=&Multiply_Vector(
     A=&SelectVectorFromArray(Direct Array Index=$Particles.PartitionParticles.Partition, Vector Selection Array=$System.Star Location)
     B=&MakeCustomVectorFromBool(
      Bool=&SetBoolByIntComparison(
       A=$Particles.PartitionParticles.Partition
       B=&Subtract_Integer(A:NiagaraInt32=&MakeIntFromFloat(Float=$Emitter.Star Amount), B:NiagaraInt32=1)
       Comparison Type=::NewEnumerator4
      )
      True Vector=<1 1 1>
      False Vector=&MakeVector(
       X=1
       Y=&Lerp_Float(
        A=1
        B=-1
        Alpha=&MakeFloatFromInt(Int=&ModuloInt(A=$Emitter.LoopCount, B=2))
       )
       Z=1
      )
     )
    )
    Position Offset Coordinate Space=::Local
    Ribbon Width=0.3
    Ribbon Width Mode=::NewEnumerator1
    Sprite Rotation Mode=::NewEnumerator2
    Sprite Size Mode=::NewEnumerator3
    Uniform Sprite Size=&SelectFloatFromArray(
     Array Sampling Mode=::NewEnumerator0
     Direct Array Index=?
     Float Selection Array=DataInterface() {propertyValues {FloatData [15,12,10]}}
     Random Seed=0
    )
    UsePositionOffset=true
    Write Material Random=true
   )
   /SetVariables(Particles.Random Direction:Vector3f=&Add_Vector(
     A=&NormalizeVector(
      Vector To Normalize=&Multiply_Vector(A=Linked() {linkedVariable {name=Particles.Position, type=NiagaraPosition}}, B=<1 1 0>)
      Safe Normalize=true
      Safe Fallback Threshold=0
      Safe Fallback Vector=<0 0 0.001>
     )
     B=&RandomVector(Vector Scale=0.25)
    )) {isSetParameters=true}
  ]
  particleUpdateScript [
   &ParticleState()
   &InterpolateOverTime(Target Float Value:NiagaraFloat=&MakeCustomFloatFromBool(Bool=$User.bisSwinging, True Float=0, False Float=1))
   &ScaleColor(
    Scale Alpha=&FloatFromCurve(
     FloatCurve=DataInterface() {propertyValues {Curve {keys [
         {arriveTangent=0.0202, interpMode=RCIM_Cubic, leaveTangent=0.0202, tangentMode=RCTM_User, time=0, value=1}
         {arriveTangent=-0.0655, interpMode=RCIM_Cubic, leaveTangent=-0.0655, tangentMode=RCTM_User, time=1, value=0}
        ]}}}
     Scale Curve=&MakeCustomFloatFromBool(
      Bool=&SetBoolByFloatComparison(A=$Particles.InterpolateOverTime.InterpolateOverTime_Float.Moving Average, B=1, Comparison Type=::NewEnumerator6)
      True Float=0
      False Float=1
     )
    )
    Scale RGB=&MakeVectorFromLinearColorRGB(Color=&ColorFromCurve(ColorCurve=DataInterface() {propertyValues {
        AlphaCurve {keys [<at 0.2404 1 RCIM_Cubic>]}
        BlueCurve {keys [<at 0 3 RCIM_Cubic>,<at 0.4179 1>]}
        GreenCurve {keys [<at 0 0.219 RCIM_Cubic>,<at 0.4179 0.37>]}
        RedCurve {keys [<at 0 3 RCIM_Cubic>,<at 0.4179 0.4021>]}
       }}))
    ScaleA=true
    ScaleRGB=true
   )
   &ScaleSpriteSize(Uniform Curve Sprite Scale=DataInterface() {propertyValues {Curve {keys [
        {arriveTangent=-3.3283, interpMode=RCIM_Cubic, leaveTangent=-3.3283, tangentMode=RCTM_User, time=0, value=1.2}
        {arriveTangent=-0.0405, interpMode=RCIM_Cubic, leaveTangent=-0.0405, tangentMode=RCTM_User, time=1, value=0.1}
       ]}}})
   &SpriteRotationRate(Rotation Rate=&RandomRangeFloat(Minimum=90, Maximum=360))
   &CameraOffset(Camera Offset Amount=1)
   /SetVariables(Particles.Position=&AddVectorToPosition(
     Position=$Particles.Initial.Position
     Vector=&Multiply_VectorByFloat(
      Vector=$Particles.Random Direction
      Float=&FloatFromCurve(
       FloatCurve=DataInterface() {propertyValues {Curve {keys [
           {arriveTangent=2.482, interpMode=RCIM_Cubic, leaveTangent=2.482, tangentMode=RCTM_User, time=0, value=0}
           {interpMode=RCIM_Cubic, tangentMode=RCTM_User, time=1, value=1}
          ]}}}
       Scale Curve=&Multiply_Float(
        A=&RandomRangeFloat(Minimum=1, Maximum=6)
        B=&MakeCustomFloatFromBool(
         Bool=&SetBoolByIntComparison(
          A=$Particles.PartitionParticles.Partition
          B=&Subtract_Integer(A:NiagaraInt32=&MakeIntFromFloat(Float=$Emitter.Star Amount), B:NiagaraInt32=1)
          Comparison Type=::NewEnumerator4
         )
         True Float=0
         False Float=1
        )
       )
      )
     )
    )) {isSetParameters=true}
   &ScaleRibbonWidth(Ribbon Width Scale=&FloatFromCurve(FloatCurve=DataInterface() {propertyValues {Curve {keys [{interpMode=RCIM_Cubic, tangentMode=RCTM_User, time=0, value=1}, {interpMode=RCIM_Cubic, tangentMode=RCTM_User, time=1, value=0.2}]}}}))
  ]
  rendererClasses [&NiagaraSpriteRendererProperties, &NiagaraRibbonRendererProperties]
  renderers [
   {rendererClass=&NiagaraSpriteRendererProperties}
   {rendererClass=&NiagaraRibbonRendererProperties, rendererIndex=1}
  ]
 }
}
summary {
 systemName=NS_Pickaxe_GalaxyKnight_IdleSharpBlade
 emitters [
  {emitterName=Sheen_Mesh, rendererClasses [&NiagaraMeshRendererProperties]}
  {emitterName=Sparkle, rendererClasses [&NiagaraSpriteRendererProperties]}
  {
   emitterName=Star_Map_Loading
   rendererClasses [&NiagaraSpriteRendererProperties, &NiagaraRibbonRendererProperties]
  }
 ]
 userVariables [User.bisSwinging {type=NiagaraBool}]
}
